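# Sound cue for the throw whoosh of the Athena smoke grenade, in the gadgets and items sound class.
# A random node picks one of three equally weighted whoosh waves and feeds it through a modulator at the root.
# It plays at slightly reduced volume with gadget-throw attenuation.
# Its own concurrency override prevents new instances once the limit is reached.
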SoundCue /Game/Athena/Sounds/Weapons/Smoke_Grenade/SC_SmokeGrenade_Athena_Throw
root=Modulator_0
nodes [
 {id=Modulator_0, children [Random_1]}
 {id=Random_1, children [WavePlayer_0,WavePlayer_1,WavePlayer_2], properties {weights [1,1,1]}}
 {id=WavePlayer_0, properties {soundWaveAssetPtr=&/Smoke_Bomb_Throw_Whoosh_03}}
 {id=WavePlayer_1, properties {soundWaveAssetPtr=&/Smoke_Bomb_Throw_Whoosh_02}}
 {id=WavePlayer_2, properties {soundWaveAssetPtr=&/Smoke_Bomb_Throw_Whoosh_01}}
]
properties {
 attenuationSettings=/Game/Sounds/Attenuation/Fort_GadgetThrow_Att
 bOverrideConcurrency=true
 maxDistance=1536
 soundClassObject=&Fort_Gadgets_Items
 volumeMultiplier=0.8
 concurrencyOverrides {resolutionRule=PreventNew}
}
