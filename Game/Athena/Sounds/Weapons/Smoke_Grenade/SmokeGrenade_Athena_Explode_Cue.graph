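# Sound cue for the smoke grenade explosion in Athena.
# A random node picks between two explosion wave variants with equal weight and passes it through a modulator at the root.
# It plays in the ranged weapon sound class with the shared smoke grenade attenuation settings.
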SoundCue /Game/Athena/Sounds/Weapons/Smoke_Grenade/SmokeGrenade_Athena_Explode_Cue
root=Modulator_0
nodes [
 {id=Modulator_0, children [Random_0]}
 {id=Random_0, children [WavePlayer_1,WavePlayer_2], properties {weights [1,1]}}
 {id=WavePlayer_1, properties {soundWaveAssetPtr=&/sfx_SmokeGrenade_Explode_02}}
 {id=WavePlayer_2, properties {soundWaveAssetPtr=&/sfx_SmokeGrenade_Explode_01}}
]
properties {
 attenuationSettings=/Game/Sounds/Attenuation/Fort_Grenade_Smoke_Att
 maxDistance=6800
 soundClassObject=&Fort_WeaponRanged
 volumeMultiplier=1
}
