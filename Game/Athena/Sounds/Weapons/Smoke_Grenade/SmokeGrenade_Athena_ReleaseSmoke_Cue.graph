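# Looping sound cue for the Athena smoke grenade's smoke release, in the ranged weapon sound class.
# Each variant mixes a main release loop and an "Inside" loop, each with its own attenuation asset and the inside layer weighted louder.
# Each mix passes through a modulator into a quality switch: the first two levels use standard waves, the last two use "_opt" waves.
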SoundCue /Game/Athena/Sounds/Weapons/Smoke_Grenade/SmokeGrenade_Athena_ReleaseSmoke_Cue
root=QualityLevel_0
nodes [
 {id=Attenuation_0, children [WavePlayer_4], properties {attenuationSettings=/Game/Sounds/Attenuation/Fort_Grenade_Smoke_Att}}
 {
  id=Attenuation_1
  children [WavePlayer_5]
  properties {attenuationSettings=/Game/Sounds/Attenuation/Fort_Grenade_Smoke_Inside_Att, attenuationOverrides {bAttenuate=false, bSpatialize=false}}
 }
 {id=Attenuation_2, children [WavePlayer_0], properties {attenuationSettings=&Fort_Grenade_Smoke_Att}}
 {
  id=Attenuation_3
  children [WavePlayer_1]
  properties {attenuationSettings=&Fort_Grenade_Smoke_Inside_Att, attenuationOverrides {bAttenuate=false, bSpatialize=false}}
 }
 {id=Mixer_0, children [Attenuation_2, Attenuation_3], properties {inputVolume [0.75,1.5]}}
 {id=Mixer_1, children [Attenuation_0, Attenuation_1], properties {inputVolume [0.75,1.5]}}
 {id=Modulator_0, children [Mixer_0]}
 {id=Modulator_1, children [Mixer_1]}
 {id=QualityLevel_0, children [Modulator_0,Modulator_0,Modulator_1,Modulator_1]}
 {id=WavePlayer_0, properties {bLooping=true, soundWaveAssetPtr=&/sfx_SmokeGrenade_ReleaseLoop_01}}
 {
  id=WavePlayer_1
  properties {bLooping=true, soundWaveAssetPtr=&/sfx_SmokeGrenade_ReleaseLoop_Inside_01}
 }
 {id=WavePlayer_4, properties {bLooping=true, soundWaveAssetPtr=&/sfx_SmokeGrenade_ReleaseLoop_01_opt}}
 {
  id=WavePlayer_5
  properties {bLooping=true, soundWaveAssetPtr=&/sfx_SmokeGrenade_ReleaseLoop_Inside_01_opt}
 }
]
properties {
 attenuationSettings=&Fort_Grenade_Smoke_Att
 maxDistance=6800
 soundClassObject=&Fort_WeaponRanged
 volumeMultiplier=1
}
